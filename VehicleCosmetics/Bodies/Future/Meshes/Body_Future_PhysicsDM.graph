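# Body_Future_PhysicsDM is a physics asset for the Future vehicle body mesh in the vehicle cosmetics content.
# It holds a single box collision body on the root bone, offset slightly forward and raised above the origin.
# It has no constraints, so the body is one rigid shape with no articulated or separately simulated parts.
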PhysicsAsset /VehicleCosmetics/Bodies/Future/Meshes/Body_Future_PhysicsDM
bodies {root [{shapeType=Box, extentX=444, extentY=216, extentZ=132, shapeName=root_box, center=<35.9698 0 91.3762>}]}
constraints []
