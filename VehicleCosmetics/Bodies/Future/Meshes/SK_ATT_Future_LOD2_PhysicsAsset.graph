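# Physics asset for the LOD2 skeletal mesh of a Future-body cosmetic vehicle attachment.
# It holds a single body named "body" with one large capsule, offset along negative X and raised above the origin.
# The asset has no constraints, so it provides only simple collision with no articulated parts.
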PhysicsAsset /VehicleCosmetics/Bodies/Future/Meshes/SK_ATT_Future_LOD2_PhysicsAsset
bodies {body [{shapeType=Capsule, length=77.0165, radius=73.315, center=<-165.4884 0.1055 85.189>, rotation=<rot 0 90.0016 -80.9732>}]}
constraints []
