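# Physics asset for the Future vehicle cosmetic body skeletal mesh SK_ATT_Future.
# It holds a single physics body made of one capsule, offset from its origin and rotated.
# No constraints are defined, so the asset provides one simple collision volume with no joints.
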PhysicsAsset /VehicleCosmetics/Bodies/Future/Meshes/SK_ATT_Future_PhysicsAsset
bodies {body [{shapeType=Capsule, length=77.8509, radius=74.7075, center=<-165.4989 0.2356 84.5644>, rotation=<rot 0.1735 -88.8686 -90>}]}
constraints []
